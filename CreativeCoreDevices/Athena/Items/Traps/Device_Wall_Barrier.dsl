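; Creative wall-placed barrier device, a child Blueprint that relies on its parent for core behavior and customizes the sizing of its barrier zone.
; On a volume size update it calls the parent, then scales BarrierBounds by the user's X/Y/Z zone options times grid-cell dimensions, slightly inset.
; GetTriggerLocation keeps the trigger's relative X and Z but derives Y from its box Z extent times the Z zone scale.
; Overlap and Tick events are empty.
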(event EventBeginPlay
  (|CallParentFunction))

(event Collision|EventActorBeginOverlap
  (params (OtherActor :type Actor%20Object%20Reference)))

(event EventTick
  (params (DeltaSeconds :type Float%20%28single-precision%29)))

(event EventUpdateVolumeSize
  (|CallParentFunction)
  (Transformation|SetWorldScale3D (Variables|Default|GetBarrierBounds) (- (* (Math|Vector|MakeVector (Variables|UserOptions|GetZoneXScale) (Variables|UserOptions|GetZoneYScale) (Variables|UserOptions|GetZoneZScale)) "512.000000,384.000000,512.000000") 5.0)))

(fn GetTriggerLocation
  (return (Math|Vector|MakeVector (.x (Variables|/Script/Engine.SceneComponent|GetRelativeLocation (Variables|Default|GetTrigger))) (* (.z (Variables|/Script/Engine.BoxComponent|GetBoxExtent (Variables|Default|GetTrigger))) (Variables|UserOptions|GetZoneZScale)) (.z (Variables|/Script/Engine.SceneComponent|GetRelativeLocation (Variables|Default|GetTrigger))))))

(fn ConstructionScript
  (|CallParentFunction))
